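# Physics asset for the hood drawstrings on the male medium M_MED_COD_Squad_Hoodie body.
# Capsules on spine_04, spine_05 and both clavicles form a coarse upper-torso volume around the strings.
# Each side has a master bone capsule freely rotating against spine_05, followed by a three-link drawstring chain.
# Chain links have moderate swing limits and very tight twist limits, so each string bends with little twisting.
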PhysicsAsset /BRCosmetics/Characters/Player/Male/Medium/Bodies/M_MED_COD_Squad_Hoodie/Meshes/M_MED_COD_Squad_Hoodie_HoodStrapsPhysics
bodies {
 clavicle_l [{shapeType=Capsule, length=5, radius=10, shapeName=clavicle_l_capsule, center=<5 0 -2>, rotation=<rot 90 0 0>}]
 clavicle_r [{shapeType=Capsule, length=5, radius=10, shapeName=clavicle_r_capsule, center=<-5 0 2>, rotation=<rot 90 0 0>}]
 drawstring_1_l [{shapeType=Capsule, length=3, radius=0.8, shapeName=drawstring_1_l_capsule, center=<3 0 0.5>, rotation=<rot 80 0 0>}]
 drawstring_1_l_MSTR [{shapeType=Capsule, length=15, radius=1, shapeName=drawstring_1_l_MSTR_capsule, center=<8 0 -2>, rotation=<rot 80 0 -5>}]
 drawstring_1_r [{shapeType=Capsule, length=3, radius=0.8, shapeName=drawstring_1_r_capsule, center=<3 0 -0.5>, rotation=<rot 95 0 0>}]
 drawstring_1_r_MSTR [{shapeType=Capsule, length=15, radius=1, shapeName=drawstring_1_r_MSTR_capsule, center=<8 0 2>, rotation=<rot 97 0 -5>}]
 drawstring_2_l [{shapeType=Capsule, length=3, radius=0.8, shapeName=drawstring_2_l_capsule, center=<3 0 0>, rotation=<rot 90 0 0>}]
 drawstring_2_r [{shapeType=Capsule, length=3, radius=0.8, shapeName=drawstring_2_r_capsule, center=<3 0 0>, rotation=<rot 90 0 0>}]
 drawstring_3_l [{shapeType=Capsule, length=3, radius=0.8, shapeName=drawstring_3_l_capsule, center=<3 0 0>, rotation=<rot 90 0 0>}]
 drawstring_3_r [{shapeType=Capsule, length=3, radius=0.8, shapeName=drawstring_3_r_capsule, center=<3 0 0>, rotation=<rot 90 0 0>}]
 spine_04 [{shapeType=Capsule, length=4, radius=13, shapeName=spine_04_capsule}]
 spine_05 [{shapeType=Capsule, length=7, radius=12.25, shapeName=spine_05_capsule, center=<-2.8557 0 0>}]
}
constraints [
 {
  bone1Name=clavicle_l
  bone2Name=spine_05
  swing1LimitDegrees=45
  swing1Motion=Free
  swing2LimitDegrees=45
  swing2Motion=Free
  twistLimitDegrees=45
  twistMotion=Free
 }
 {
  bone1Name=clavicle_r
  bone2Name=spine_05
  swing1LimitDegrees=45
  swing1Motion=Free
  swing2LimitDegrees=45
  swing2Motion=Free
  twistLimitDegrees=45
  twistMotion=Free
 }
 {
  bone1Name=spine_05
  bone2Name=spine_04
  swing1LimitDegrees=45
  swing1Motion=Free
  swing2LimitDegrees=45
  swing2Motion=Free
  twistLimitDegrees=45
  twistMotion=Free
 }
 {
  bone1Name=drawstring_1_l_MSTR
  bone2Name=spine_05
  swing1LimitDegrees=45
  swing1Motion=Free
  swing2LimitDegrees=45
  swing2Motion=Free
  twistLimitDegrees=45
  twistMotion=Free
 }
 {bone1Name=drawstring_1_l, bone2Name=drawstring_1_l_MSTR, swing1LimitDegrees=25, swing2LimitDegrees=25, twistLimitDegrees=2}
 {bone1Name=drawstring_2_l, bone2Name=drawstring_1_l, swing1LimitDegrees=25, swing2LimitDegrees=25, twistLimitDegrees=2}
 {bone1Name=drawstring_3_l, bone2Name=drawstring_2_l, swing1LimitDegrees=25, swing2LimitDegrees=25, twistLimitDegrees=2}
 {
  bone1Name=drawstring_1_r_MSTR
  bone2Name=spine_05
  swing1LimitDegrees=45
  swing1Motion=Free
  swing2LimitDegrees=45
  swing2Motion=Free
  twistLimitDegrees=45
  twistMotion=Free
 }
 {bone1Name=drawstring_1_r, bone2Name=drawstring_1_r_MSTR, swing1LimitDegrees=25, swing2LimitDegrees=25, twistLimitDegrees=2}
 {bone1Name=drawstring_2_r, bone2Name=drawstring_1_r, swing1LimitDegrees=25, swing2LimitDegrees=25, twistLimitDegrees=2}
 {bone1Name=drawstring_3_r, bone2Name=drawstring_2_r, swing1LimitDegrees=25, swing2LimitDegrees=25, twistLimitDegrees=2}
]
