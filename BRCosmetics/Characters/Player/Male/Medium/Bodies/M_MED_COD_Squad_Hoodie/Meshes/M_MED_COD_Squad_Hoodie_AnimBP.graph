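# Post-process-style AnimBP for the male medium Squad Hoodie body, copying the attached parent's pose and adding secondary cloth and accessory motion.
# Control Rig feeds RigidBody hood-strap physics, then AnimDynamics for pouch, hood and hem, thigh-driven hem corrections and a shell-band constraint.
# The event graph updates gravity override, thigh-rotation alphas and pistol/rifle aiming alphas for the rig.
# A separate AnimDynamics drawstring chain never reaches the output pose.
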AnimBlueprint /BRCosmetics/Characters/Player/Male/Medium/Bodies/M_MED_COD_Squad_Hoodie/Meshes/M_MED_COD_Squad_Hoodie_AnimBP
target_skeleton=&/M_MED_COD_Squad_Hoodie_Skeleton
graphs [
 AnimGraph {nodes [
   n1 = Anim Dynamics - Bone: drawstring_1_l(ComponentPose=Local To Component(LocalPose=Copy Pose From Mesh() {bUseAttachedParent=true}), Alpha=1, ExternalForce="0,0,0") {
    angularDampingOverride=1
    angularSpringConstant=1500
    bAngularSpring=true
    bGravityOverrideInSimSpace=true
    bOverrideAngularDamping=true
    gravityScale=0.35
    simulationSpace=RootRelative
    alphaBoolBlend {blendOption=Linear}
    alphaScaleBiasClamp {scale=0.5}
    boundBone {boneName=drawstring_1_l}
    componentAppliedLinearAccClamp {z=5000}
    componentLinearAccScale {z=0.44}
    componentLinearVelScale {z=0.44}
    physicsBodyDefinitions [{
      sphereCollisionRadius=0
      boundBone {boneName=drawstring_1_l}
      boxExtents=<10 20 20>
      constraintSetup {angularLimitsMax=<0 0 0>, angularLimitsMin=<0 -20 -40>, angularTarget=<0 0 0>}
      localJointOffset=<10 0 0>
     }]
   }
   n2 = Anim Dynamics - Bone: drawstring_2_l(ComponentPose=@n1, Alpha=1, ExternalForce="0,0,0") {
    angularDampingOverride=1
    angularSpringConstant=1500
    bAngularSpring=true
    bGravityOverrideInSimSpace=true
    bOverrideAngularDamping=true
    gravityScale=0.35
    simulationSpace=RootRelative
    alphaBoolBlend {blendOption=Linear}
    alphaScaleBiasClamp {scale=0.7}
    boundBone {boneName=drawstring_2_l}
    componentAppliedLinearAccClamp {z=5000}
    componentLinearAccScale {z=0.44}
    componentLinearVelScale {z=0.44}
    physicsBodyDefinitions [{
      sphereCollisionRadius=0
      boundBone {boneName=drawstring_2_l}
      boxExtents=<10 20 20>
      constraintSetup {angularLimitsMax=<0 0 0>, angularLimitsMin=<0 -20 -40>, angularTarget=<0 0 0>}
      localJointOffset=<10 0 0>
     }]
   }
   n3 = Anim Dynamics - Bone: drawstring_3_l(ComponentPose=@n2, Alpha=1, ExternalForce="0,0,0") {
    angularDampingOverride=1
    angularSpringConstant=1500
    bAngularSpring=true
    bGravityOverrideInSimSpace=true
    bOverrideAngularDamping=true
    gravityScale=0.35
    simulationSpace=RootRelative
    alphaBoolBlend {blendOption=Linear}
    boundBone {boneName=drawstring_3_l}
    componentAppliedLinearAccClamp {z=5000}
    componentLinearAccScale {z=0.44}
    componentLinearVelScale {z=0.44}
    physicsBodyDefinitions [{
      sphereCollisionRadius=0
      boundBone {boneName=drawstring_3_l}
      boxExtents=<10 20 20>
      constraintSetup {angularLimitsMax=<0 0 0>, angularLimitsMin=<0 -20 -40>, angularTarget=<0 0 0>}
      localJointOffset=<10 0 0>
     }]
   }
   n4 = Anim Dynamics - Bone: drawstring_1_r(ComponentPose=@n3, Alpha=1, ExternalForce="0,0,0") {
    angularDampingOverride=1
    angularSpringConstant=1500
    bAngularSpring=true
    bGravityOverrideInSimSpace=true
    bOverrideAngularDamping=true
    gravityScale=0.35
    simulationSpace=RootRelative
    alphaBoolBlend {blendOption=Linear}
    alphaScaleBiasClamp {scale=0.5}
    boundBone {boneName=drawstring_1_r}
    componentAppliedLinearAccClamp {z=5000}
    componentLinearAccScale {z=0.44}
    componentLinearVelScale {z=0.44}
    physicsBodyDefinitions [{
      sphereCollisionRadius=0
      boundBone {boneName=drawstring_1_r}
      boxExtents=<10 20 20>
      constraintSetup {angularLimitsMax=<0 20 0>, angularLimitsMin=<0 0 -40>, angularTarget=<0 0 0>}
      localJointOffset=<10 0 0>
     }]
   }
   n5 = Anim Dynamics - Bone: drawstring_2_r(ComponentPose=@n4, Alpha=1, ExternalForce="0,0,0") {
    angularDampingOverride=1
    angularSpringConstant=1500
    bAngularSpring=true
    bGravityOverrideInSimSpace=true
    bOverrideAngularDamping=true
    gravityScale=0.35
    simulationSpace=RootRelative
    alphaBoolBlend {blendOption=Linear}
    alphaScaleBiasClamp {scale=0.7}
    boundBone {boneName=drawstring_2_r}
    componentAppliedLinearAccClamp {z=5000}
    componentLinearAccScale {z=0.44}
    componentLinearVelScale {z=0.44}
    physicsBodyDefinitions [{
      sphereCollisionRadius=0
      boundBone {boneName=drawstring_2_r}
      boxExtents=<10 20 20>
      constraintSetup {angularLimitsMax=<0 20 0>, angularLimitsMin=<0 0 -40>, angularTarget=<0 0 0>}
      localJointOffset=<10 0 0>
     }]
   }
   n6 = Anim Dynamics - Bone: drawstring_3_r(ComponentPose=@n5, Alpha=1, ExternalForce="0,0,0") {
    angularDampingOverride=1
    angularSpringConstant=1500
    bAngularSpring=true
    bGravityOverrideInSimSpace=true
    bOverrideAngularDamping=true
    gravityScale=0.35
    simulationSpace=RootRelative
    alphaBoolBlend {blendOption=Linear}
    boundBone {boneName=drawstring_3_r}
    componentAppliedLinearAccClamp {z=5000}
    componentLinearAccScale {z=0.44}
    componentLinearVelScale {z=0.44}
    physicsBodyDefinitions [{
      sphereCollisionRadius=0
      boundBone {boneName=drawstring_3_r}
      boxExtents=<10 20 20>
      constraintSetup {angularLimitsMax=<0 20 0>, angularLimitsMin=<0 0 -40>, angularTarget=<0 0 0>}
      localJointOffset=<10 0 0>
     }]
   }
   Blend Poses by bool(BlendPose_1=Component To Local(ComponentPose=@n6), bActiveValue=Get bRigidBodyEnabled()) {blendType=Linear}
   n7 = Control Rig!(
    Source=Copy Pose From Mesh() {bUseAttachedParent=true}
    Alpha
    Is Pistol Targeting:double=Get Is Pistol Targeting Alpha() -> (Is Pistol Targeting Alpha:float)
    Is Rifle Targeting:double=Get Is Rifle Targeting Alpha() -> (Is Rifle Targeting Alpha:float)
    Is Rifle:double=Get Is Rifle Alpha() -> (Is Rifle Alpha:float)
   ) -> (Pose) {
    alphaBoolBlend {blendOption=Linear}
    controlRigAssetReference {blueprintRigClass=&/M_MED_COD_Squad_Hoodie_CtrlRig.M_MED_COD_Squad_Hoodie_CtrlRig_C}
   }
   n8 = RigidBody!(ComponentPose=Local To Component(LocalPose=@n7), Alpha, OverrideWorldGravity:Vector=Get GravityOverride(), ExternalForce) -> (Pose) {
    bClampLinearTranslationLimitToRefPose=true
    bOverrideWorldGravity=true
    overridePhysicsAsset=&/M_MED_COD_Squad_Hoodie_HoodStrapsPhysics
    simulationSpace=BaseBoneSpace
    alphaBoolBlend {blendOption=Linear}
    alphaScaleBiasClamp {interpSpeedIncreasing=2}
    baseBoneRef {boneName=spine_05}
    componentAppliedLinearAccClamp=<3000 3000 3000>
    componentLinearAccScale=<0.4 0.3 0.3>
    componentLinearVelScale=<0.4 0.3 0.3>
    simSpaceSettings {
     maxAngularAcceleration=1000
     maxAngularVelocity=3
     maxLinearAcceleration=700
     maxLinearVelocity=500
     velocityScaleZ=0
     worldAlpha=0.25
     externalAngularVelocity {z=0.1}
     externalLinearDragV=<0.2 0.2>
    }
   }
   n9 = Anim Dynamics - Bone: thighPouch(ComponentPose=Local To Component(LocalPose=Component To Local(ComponentPose=@n8)), Alpha=1, ExternalForce="0,0,0") {
    angularDampingOverride=1
    angularSpringConstant=1500
    bAngularSpring=true
    bGravityOverrideInSimSpace=true
    bOverrideAngularDamping=true
    gravityScale=0.35
    simulationSpace=RootRelative
    alphaBoolBlend {blendOption=Linear}
    alphaScaleBiasClamp {scale=0.6}
    boundBone {boneName=thighPouch}
    componentAppliedLinearAccClamp {z=5000}
    physicsBodyDefinitions [{
      sphereCollisionRadius=0
      boundBone {boneName=thighPouch}
      boxExtents=<10 20 20>
      constraintSetup {angularLimitsMax=<0 0 0>, angularLimitsMin=<0 0 -40>, angularTarget=<0 0 0>}
      localJointOffset=<10 0 0>
     }]
   }
   n10 = Anim Dynamics - Bone: hood_l(ComponentPose=@n9, Alpha=1, ExternalForce="0,0,0") {
    angularDampingOverride=1
    angularSpringConstant=1500
    bAngularSpring=true
    bGravityOverrideInSimSpace=true
    bOverrideAngularDamping=true
    gravityScale=0.35
    simulationSpace=RootRelative
    alphaBoolBlend {blendOption=Linear}
    alphaScaleBiasClamp {scale=0.15}
    boundBone {boneName=hood_l}
    componentAppliedLinearAccClamp {z=5000}
    physicsBodyDefinitions [{
      sphereCollisionRadius=0
      boundBone {boneName=hood_l}
      boxExtents=<10 20 20>
      constraintSetup {angularLimitsMax=<0 25 0>, angularLimitsMin=<0 -25 -40>, angularTarget=<1 0 0>}
      localJointOffset=<10 0 0>
     }]
   }
   n11 = Anim Dynamics - Bone: hood_r(ComponentPose=@n10, Alpha=1, ExternalForce="0,0,0") {
    angularDampingOverride=1
    angularSpringConstant=1500
    bAngularSpring=true
    bGravityOverrideInSimSpace=true
    bOverrideAngularDamping=true
    gravityScale=0.35
    simulationSpace=RootRelative
    alphaBoolBlend {blendOption=Linear}
    alphaScaleBiasClamp {scale=0.15}
    boundBone {boneName=hood_r}
    componentAppliedLinearAccClamp {z=5000}
    physicsBodyDefinitions [{
      sphereCollisionRadius=0
      boundBone {boneName=hood_r}
      boxExtents=<10 20 20>
      constraintSetup {angularLimitsMax=<0 25 0>, angularLimitsMin=<0 -25 -40>, angularTarget=<1 0 0>}
      localJointOffset=<10 0 0>
     }]
   }
   n12 = Anim Dynamics - Bone: hood_c(ComponentPose=@n11, Alpha=1, ExternalForce="0,0,0") {
    angularDampingOverride=1
    angularSpringConstant=1500
    bAngularSpring=true
    bGravityOverrideInSimSpace=true
    bOverrideAngularDamping=true
    gravityScale=0.35
    simulationSpace=RootRelative
    alphaBoolBlend {blendOption=Linear}
    alphaScaleBiasClamp {scale=0.08}
    boundBone {boneName=hood_c}
    componentAppliedLinearAccClamp {z=5000}
    componentLinearAccScale {z=0.4}
    componentLinearVelScale {z=0.4}
    physicsBodyDefinitions [{
      sphereCollisionRadius=0
      boundBone {boneName=hood_c}
      boxExtents=<10 20 20>
      constraintSetup {angularLimitsMax=<30 60 80>, angularLimitsMin=<-30 -60 -90>, angularTarget=<1 0 0>}
      localJointOffset=<10 0 0>
     }]
   }
   n13 = Anim Dynamics - Bone: jacket_pelvis_r(ComponentPose=@n12, Alpha=1, ExternalForce="0,0,0") {
    angularDampingOverride=1
    angularSpringConstant=1500
    bAngularSpring=true
    bGravityOverrideInSimSpace=true
    bOverrideAngularDamping=true
    gravityScale=0.35
    simulationSpace=RootRelative
    alphaBoolBlend {blendOption=Linear}
    alphaScaleBiasClamp {scale=0.3}
    boundBone {boneName=jacket_pelvis_r}
    componentAppliedLinearAccClamp=<5000 5000 5000>
    componentLinearAccScale {z=0.44}
    componentLinearVelScale {z=0.44}
    physicsBodyDefinitions [{
      sphereCollisionRadius=0
      boundBone {boneName=jacket_pelvis_r}
      boxExtents=<10 20 20>
      constraintSetup {angularLimitsMax=<0 30 0>, angularLimitsMin=<0 -30 -60>, angularTarget=<0 0 0>}
      localJointOffset=<10 0 0>
     }]
   }
   n14 = Anim Dynamics - Bone: jacket_pelvis_l(ComponentPose=@n13, Alpha=1, ExternalForce="0,0,0") {
    angularDampingOverride=1
    angularSpringConstant=1500
    bAngularSpring=true
    bGravityOverrideInSimSpace=true
    bOverrideAngularDamping=true
    gravityScale=0.35
    simulationSpace=RootRelative
    alphaBoolBlend {blendOption=Linear}
    alphaScaleBiasClamp {scale=0.3}
    boundBone {boneName=jacket_pelvis_l}
    componentAppliedLinearAccClamp=<5000 5000 5000>
    componentLinearAccScale {z=0.44}
    componentLinearVelScale {z=0.44}
    physicsBodyDefinitions [{
      sphereCollisionRadius=0
      boundBone {boneName=jacket_pelvis_l}
      boxExtents=<10 20 20>
      constraintSetup {angularLimitsMax=<0 30 0>, angularLimitsMin=<0 -30 -60>, angularTarget=<0 0 0>}
      localJointOffset=<10 0 0>
     }]
   }
   n15 = Anim Dynamics - Bone: jacket_pelvis_c(ComponentPose=@n14, Alpha=1, ExternalForce="0,0,0") {
    angularDampingOverride=1
    angularSpringConstant=1500
    bAngularSpring=true
    bGravityOverrideInSimSpace=true
    bOverrideAngularDamping=true
    gravityScale=0.35
    simulationSpace=RootRelative
    alphaBoolBlend {blendOption=Linear}
    alphaScaleBiasClamp {scale=0.3}
    boundBone {boneName=jacket_pelvis_c}
    componentAppliedLinearAccClamp=<5000 5000 5000>
    componentLinearAccScale {z=0.45}
    componentLinearVelScale {z=0.45}
    physicsBodyDefinitions [{
      sphereCollisionRadius=0
      boundBone {boneName=jacket_pelvis_c}
      boxExtents=<10 20 20>
      constraintSetup {angularLimitsMax=<0 30 0>, angularLimitsMin=<0 -30 -60>, angularTarget=<0 0 0>}
      localJointOffset=<10 0 0>
     }]
   }
   n16 = Anim Dynamics - Bone: jacket_pelvis_fnt_r(ComponentPose=@n15, Alpha=1, ExternalForce="0,0,0") {
    angularDampingOverride=1
    angularSpringConstant=1500
    bAngularSpring=true
    bGravityOverrideInSimSpace=true
    bOverrideAngularDamping=true
    gravityScale=0.35
    simulationSpace=RootRelative
    alphaBoolBlend {blendOption=Linear}
    alphaScaleBiasClamp {scale=0.45}
    boundBone {boneName=jacket_pelvis_fnt_r}
    componentAppliedLinearAccClamp=<5000 5000 5000>
    componentLinearAccScale {z=0.44}
    componentLinearVelScale {z=0.44}
    physicsBodyDefinitions [{
      collisionType=CustomSphere
      sphereCollisionRadius=3
      boundBone {boneName=jacket_pelvis_fnt_r}
      boxExtents=<10 20 20>
      constraintSetup {angularLimitsMax=<0 20 0>, angularLimitsMin=<0 -20 -40>, angularTarget=<0 0 0>}
      localJointOffset=<10 0 0>
     }]
    sphericalLimits [
     {limitRadius=12, drivingBone {boneName=thigh_r}, sphereLocalOffset=<15 0 0>}
     {limitRadius=12, drivingBone {boneName=thigh_r}, sphereLocalOffset=<0 0 0>}
    ]
   }
   n17 = Anim Dynamics - Bone: jacket_pelvis_fnt_2_r(ComponentPose=@n16, Alpha=1, ExternalForce="0,0,0") {
    angularDampingOverride=1
    angularSpringConstant=1500
    bAngularSpring=true
    bGravityOverrideInSimSpace=true
    bOverrideAngularDamping=true
    gravityScale=0.35
    simulationSpace=RootRelative
    alphaBoolBlend {blendOption=Linear}
    alphaScaleBiasClamp {scale=0.45}
    boundBone {boneName=jacket_pelvis_fnt_2_r}
    componentAppliedLinearAccClamp=<5000 5000 5000>
    componentLinearAccScale {z=0.44}
    componentLinearVelScale {z=0.44}
    physicsBodyDefinitions [{
      collisionType=CustomSphere
      sphereCollisionRadius=3
      boundBone {boneName=jacket_pelvis_fnt_2_r}
      boxExtents=<10 20 20>
      constraintSetup {angularLimitsMax=<0 20 0>, angularLimitsMin=<0 -20 -40>, angularTarget=<0 0 0>}
      localJointOffset=<10 0 0>
     }]
    sphericalLimits [
     {limitRadius=12, drivingBone {boneName=thigh_r}, sphereLocalOffset=<15 0 0>}
     {limitRadius=12, drivingBone {boneName=thigh_r}, sphereLocalOffset=<0 0 0>}
    ]
   }
   n18 = Anim Dynamics - Bone: jacket_pelvis_fnt_l(ComponentPose=@n17, Alpha=0, ExternalForce="0,0,0") {
    angularDampingOverride=1
    angularSpringConstant=1500
    bAngularSpring=true
    bGravityOverrideInSimSpace=true
    bOverrideAngularDamping=true
    gravityScale=0.35
    simulationSpace=RootRelative
    alphaBoolBlend {blendOption=Linear}
    alphaScaleBiasClamp {scale=0.45}
    boundBone {boneName=jacket_pelvis_fnt_l}
    componentAppliedLinearAccClamp=<5000 5000 5000>
    componentLinearAccScale {z=0.44}
    componentLinearVelScale {z=0.44}
    physicsBodyDefinitions [{
      collisionType=CustomSphere
      sphereCollisionRadius=3
      boundBone {boneName=jacket_pelvis_fnt_l}
      boxExtents=<10 20 20>
      constraintSetup {angularLimitsMax=<0 0 0>, angularLimitsMin=<0 0 -150>, angularTarget=<0 0 0>}
      localJointOffset=<10 0 0>
     }]
    sphericalLimits [
     {limitRadius=12, drivingBone {boneName=thigh_l}, sphereLocalOffset=<-15 0 0>}
     {limitRadius=12, drivingBone {boneName=thigh_l}, sphereLocalOffset=<0 0 0>}
    ]
   }
   n19 = "Transform (Modify) Bone - Bone: jacket_pelvis_fnt_l"(ComponentPose=@n18, Alpha=Get thigh_rot_l(), Translation="10,3,0", Rotation="45,-45.000,0.000", Scale="0.6,1,1") {
    rotationMode=BMM_Additive
    rotationSpace=BCS_ParentBoneSpace
    scaleMode=BMM_Additive
    translationMode=BMM_Additive
    translationSpace=BCS_ParentBoneSpace
    alphaBoolBlend {blendOption=Linear}
    alphaScaleBiasClamp {bInterpResult=true, interpSpeedDecreasing=20, interpSpeedIncreasing=20, scale=0.9}
    boneToModify {boneName=jacket_pelvis_fnt_l}
    rotation {pitch=45, yaw=-45}
    scale {x=0.6}
    translation=<10 3>
   }
   n20 = Look At - Bone: jacket_pelvis_fnt_r(ComponentPose=@n19, Alpha=0.8) {
    alphaBoolBlend {blendOption=Linear}
    boneToModify {boneName=jacket_pelvis_fnt_r}
    lookAtLocation {x=0}
    lookAtTarget {boneReference {boneName=jacket_pelvis_fnt_tgt}}
    lookAt_Axis {axis=<1 0>}
   }
   n21 = Look At - Bone: jacket_pelvis_fnt_2_r(ComponentPose=@n20, Alpha=0.8) {
    alphaBoolBlend {blendOption=Linear}
    boneToModify {boneName=jacket_pelvis_fnt_2_r}
    lookAtLocation {x=0}
    lookAtTarget {boneReference {boneName=jacket_pelvis_fnt_2_tgt}}
    lookAt_Axis {axis=<1 0>}
   }
   n22 = Get thigh_rot_r()
   n23 = "Transform (Modify) Bone - Bone: jacket_pelvis_fnt_r"(ComponentPose=@n21, Alpha=@n22, Translation="8,3,1", Rotation="0,-5,0", Scale="0.7,1,1") {
    rotationMode=BMM_Additive
    rotationSpace=BCS_ParentBoneSpace
    scaleMode=BMM_Replace
    translationMode=BMM_Additive
    translationSpace=BCS_ParentBoneSpace
    alphaBoolBlend {blendOption=Linear}
    alphaScaleBiasClamp {bInterpResult=true}
    boneToModify {boneName=jacket_pelvis_fnt_r}
    rotation {yaw=-5}
    scale {x=0.7}
    translation=<8 3 1>
   }
   n24 = "Transform (Modify) Bone - Bone: jacket_pelvis_fnt_2_r"(ComponentPose=@n23, Alpha=@n22, Translation="3,0,0", Rotation="0,-5,0", Scale="0.7,1,1") {
    rotationMode=BMM_Additive
    rotationSpace=BCS_ParentBoneSpace
    scaleMode=BMM_Replace
    translationMode=BMM_Additive
    translationSpace=BCS_ParentBoneSpace
    alphaBoolBlend {blendOption=Linear}
    alphaScaleBiasClamp {bInterpResult=true}
    boneToModify {boneName=jacket_pelvis_fnt_2_r}
    rotation {yaw=-5}
    scale {x=0.7}
    translation {x=3}
   }
   n25 = "Constraint - shellBand_r (PPPP)"!(
    ComponentPose=@n24
    Alpha=1
    ConstraintWeights_0=0.29
    ConstraintWeights_1:float=0.29
    ConstraintWeights_2:float=0.38
    ConstraintWeights_3:float=0.04
   ) -> (Pose) {
    alphaBoolBlend {blendOption=Linear}
    boneToModify {boneName=shellBand_r}
    constraintSetup [
     {targetBone {boneName=upperarm_r}}
     {targetBone {boneName=upperarm_twist_01_r}}
     {targetBone {boneName=upperarm_twist_02_r}}
     {targetBone {boneName=clavicle_r}}
    ]
    constraintWeights [1,1,1,1]
   }
   Output Pose(Result=Component To Local(ComponentPose=@n25)) {layerGroup=None}
  ]}
 EventGraph {nodes [
   Event BlueprintUpdateAnimation()
   > n1 = Sequence!(execute) -> (then_0, then_1, then_2:Exec)
   > n2 = Gravity Override State Function DEPRECATED(CharacterPartAnimInstance=Self-Reference(), GameTimeInSeconds=GetGameTimeInSeconds(), GravityOverrideParams=Get Gravity Override Params())
   > Set GravityOverride(GravityOverride=@n2)
   n3 = BreakTransform(InTransform=GetDeltaTransformFromRefPose(self=GetOwnerSkeletalMeshComponent(), BoneName=thigh_r, BaseName=pelvis))
   Set thigh_rot_r(execute=@n1.then_1, thigh_rot_r=MapRangeClamped(Value=BreakRotator(InRot=@n3.Rotation).Yaw, InRangeA=0, InRangeB=-110))
   n4 = BreakTransform(InTransform=GetDeltaTransformFromRefPose(self=GetOwnerSkeletalMeshComponent(), BoneName=thigh_l, BaseName=pelvis))
   > Set thigh_rot_l(thigh_rot_l=MapRangeClamped(Value=BreakRotator(InRot=@n4.Rotation).Yaw, InRangeA=-15, InRangeB=-110))
   n5 = Sequence!(execute=@n1.then_2) -> (then_0, then_1, then_2, then_3:Exec, then_4:Exec)
   > n6 = Cast To FortPlayerAnimInstance(Object=GetAnimInstance(self=GetOwnerSkeletalMeshComponent()))
   > Set isPistol(isPistol=false)
   > Set isRifle(isRifle:bool=false) -> (Output_Get:bool)
   > n7 = Switch on EFortWeaponCoreAnimation!(execute, Selection:EFortWeaponCoreAnimation Enum=Get WeaponCoreAnim(self=@n6)) -> (
    Melee:Exec
    Pistol:Exec
    Shotgun:Exec
    PaperBlueprint:Exec
    Rifle:Exec
    MeleeOneHand:Exec
    MachinePistol:Exec
    RocketLauncher:Exec
    GrenadeLauncher:Exec
    GoingCommando:Exec
    AssaultRifle:Exec
    TacticalShotgun:Exec
    SniperRifle:Exec
    TrapPlacement:Exec
    ShoulderLauncher:Exec
    AbilityDecoTool:Exec
    Crossbow:Exec
    C4:Exec
    RemoteControl:Exec
    DualWield:Exec
    AR_BullPup:Exec
    AR_ForwardGrip:Exec
    MedPackPaddles:Exec
    SMG_P90:Exec
    AR_DrumGun:Exec
    Consumable_Small:Exec
    Consumable_Large:Exec
    Balloon:Exec
    MountedTurret:Exec
    CreativeTool:Exec
    ExplosiveBow:Exec
    AshtonIndigo:Exec
    AshtonChicago:Exec
    MeleeDualWield:Exec
    Unarmed:Exec
   )
   Set isRifle(execute=[@n7.AssaultRifle, @n7.Rifle, @n7.RocketLauncher, @n7.Shotgun, @n7.SniperRifle, @n7.TacticalShotgun], isRifle=true)
   Set isPistol(execute=@n7.Pistol, isPistol=true)
   n8 = Get isRifle() -> (isRifle:bool)
   n9 = Get bIsTargeting()
   Set Is Rifle Alpha(execute=@n5.then_1, Is Rifle Alpha:float="To Float (Boolean)"(InBool=AND Boolean(A=@n8, B=@n9))) -> (Output_Get:float)
   Set Is Pistol Alpha(execute=@n5.then_2, Is Pistol Alpha:float="To Float (Boolean)"(InBool=Get isPistol() -> (isPistol:bool))) -> (Output_Get:float)
   n10 = Get isPistol()
   n11 = Get bIsTargeting()
   Set Is Pistol Targeting Alpha(execute=@n5.then_3, Is Pistol Targeting Alpha:float="To Float (Boolean)"(InBool=AND Boolean(A=@n10, B=@n11))) -> (Output_Get:float)
   n12 = Get bIsTargeting()
   n13 = Get isRifle()
   Set Is Rifle Targeting Alpha(execute=@n5.then_4, Is Rifle Targeting Alpha:float="To Float (Boolean)"(InBool=AND Boolean(A=@n13, B=@n12))) -> (Output_Get:float)
  ]}
]
variables [
 Gravity Override Params {type=GravityOverrideParamsStruct}
 GravityOverride {type=Vector}
 Is Pistol Alpha {type=number}
 Is Pistol Targeting Alpha {type=number}
 Is Rifle Alpha {type=number}
 Is Rifle Targeting Alpha {type=number}
 isPistol {type=boolean}
 isRifle {type=boolean}
 thigh_rot_l {type=number}
 thigh_rot_r {type=number}
]
